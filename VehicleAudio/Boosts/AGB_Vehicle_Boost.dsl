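; Audio gameplay behavior for vehicle boost sounds, with a start one-shot, a loop and an end sound.
; On start it plays the start and looping sounds, choosing default or custom assets via ShouldPlayDefaultSound and passing the IsLocal parameter.
; It stores each voice's component and a playing flag.
; On stop it plays the end sound, then stops both stored components with their fade-out times and clears the flags.
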(event AudioGameplayBehavior|EventOnStart
  (Utilities|FlowControl|Sequence
    (:then_0
      ;[1 1 1] Start Sound
      (Variables|Default|GetStartSoundComponent
        (:then
          (if (Variables|Default|GetbIsStartPlaying)
            _
            (else
              _)))
        (:else
          _))
      (bind _returnvalue (AudioGameplayBehavior|PlaySound (select (Utilities|ShouldPlayDefaultSound) (Variables|Default|GetDefaultStartSound) (Variables|Default|GetStartSound)) 0.0 1.0 0.0 (AudioGameplayBehavior|GetBoolParamValue "IsLocal")))
      (bind (_returnvalue_sound _returnvalue_component) (Utilities|Struct|BreakActiveVoice _returnvalue))
      (Variables|Default|SetStartSoundComponent _returnvalue_component)
      (Variables|Default|SetbIsStartPlaying (Utilities|IsValid _returnvalue_sound :!class /Script/Engine.KismetSystemLibrary)))
    (:then_1
      ;[1 1 1] Looping Sound
      (bind _loopingsound_component (Variables|Default|GetLoopingSound%20Component)
        (:then
          (if (Variables|Default|GetbIsLoopPlaying)
            _
            (else
              _)))
        (:else
          _))
      (bind _returnvalue_2 (AudioGameplayBehavior|PlaySound (select (Utilities|ShouldPlayDefaultSound) (Variables|Default|GetDefaultLoopingSound) (Variables|Default|GetLoopingSound)) 0.0 1.0 0.0 (AudioGameplayBehavior|GetBoolParamValue "IsLocal")))
      (bind (_returnvalue_sound_2 _returnvalue_component_2) (Utilities|Struct|BreakActiveVoice _returnvalue_2))
      (Variables|Default|SetLoopingSound%20Component _returnvalue_component_2)
      (Variables|Default|SetbIsLoopPlaying (Utilities|IsValid _returnvalue_sound_2 :!class /Script/Engine.KismetSystemLibrary)))))

(event AudioGameplayBehavior|EventOnStop
  (Utilities|FlowControl|Sequence
    (:then_0
      ;[1 1 1] play end sound
      (AudioGameplayBehavior|PlaySound (select (Utilities|ShouldPlayDefaultSound) (Variables|Default|GetDefaultEndSound) (Variables|Default|GetEndSound))))
    (:then_1
      ;[1 1 1] Stop the previously playing sounds
      (bind _loopingsound_component (Variables|Default|GetLoopingSound%20Component)
        (:then
          (AudioGameplayBehavior|StopComponent _loopingsound_component (Variables|Default|GetLoopFadeoutTime))
          (Variables|Default|SetbIsLoopPlaying false))
        (:else)))
    (:then_2
      (bind _startsoundcomponent (Variables|Default|GetStartSoundComponent)
        (:then
          (AudioGameplayBehavior|StopComponent _startsoundcomponent (Variables|Default|GetStartFadeout%20Time))
          (Variables|Default|SetbIsStartPlaying false))
        (:else)))))
