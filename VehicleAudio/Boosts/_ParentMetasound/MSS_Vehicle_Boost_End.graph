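# Parent MetaSound source for the stereo one-shot that plays when a vehicle boost ends.
# On play it picks a shuffled wave from the input array, scales it by a decibel volume input and signals finished when the wave ends.
# Pitch shift follows quartic-eased normal speed mapped between min and max inputs, multiplied to zero when the pitch-by-speed toggle is off.
# Boost modulation patches drive volume and high/low-pass filters.
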MetaSoundSource /VehicleAudio/Boosts/_ParentMetasound/MSS_Vehicle_Boost_End
inputs [
 Boost.End.Volume {type=Float, access=Constructor}
 Boost.End.WaveAsset {type=WaveAsset:Array, access=Constructor}
 Boost.Pitch.Max {type=Float, access=Constructor}
 Boost.Pitch.Min {type=Float, access=Constructor}
 Boost.PitchByNormalSpeed {type=Bool}
 Fort.Vehicle.Acceleration {type=Float}
 Fort.Vehicle.BlowoffStrength {type=Float}
 Fort.Vehicle.BonusSpeed {type=Float}
 Fort.Vehicle.EnableShifting {type=Bool}
 Fort.Vehicle.Gear {type=Int32}
 Fort.Vehicle.IsGrounded {type=Bool}
 Fort.Vehicle.Load {type=Float}
 Fort.Vehicle.NormalSpeed {type=Float}
 Fort.Vehicle.OnBlowoff {type=Trigger}
 Fort.Vehicle.OnDownshift {type=Trigger}
 Fort.Vehicle.OnUpshift {type=Trigger}
 Fort.Vehicle.Pitch {type=Float}
 Fort.Vehicle.RPM {type=Float}
 Fort.Vehicle.Throttle {type=Float}
 UE.Source.OnPlay {type=Trigger}
 UE.Source.Orientation.Azimuth {type=Float}
 UE.Source.Orientation.Elevation {type=Float}
]
interfaces [Fort.Vehicle:1.0, UE.OutputFormat.Stereo:1.0, UE.Source.OneShot:1.0, UE.Source.Orientation:1.0, UE.Source:1.1]
outputs [
 "UE.OutputFormat.Stereo.Audio:0" {type=Audio}
 "UE.OutputFormat.Stereo.Audio:1" {type=Audio}
 UE.Source.OneShot.OnFinished {type=Trigger}
]
pages [Default {
  input_defaults {
   Boost.End.Volume=0
   Boost.Pitch.Max=2.5
   Boost.Pitch.Min=0
   Boost.PitchByNormalSpeed="<MetaSoundBuilderResult.SUCCEEDED: 0>"
   Fort.Vehicle.Acceleration=0
   Fort.Vehicle.BlowoffStrength=0
   Fort.Vehicle.BonusSpeed=0
   Fort.Vehicle.EnableShifting=null
   Fort.Vehicle.Gear=0
   Fort.Vehicle.IsGrounded="<MetaSoundBuilderResult.SUCCEEDED: 0>"
   Fort.Vehicle.Load=0
   Fort.Vehicle.NormalSpeed=0
   Fort.Vehicle.OnBlowoff=null
   Fort.Vehicle.OnDownshift=null
   Fort.Vehicle.OnUpshift=null
   Fort.Vehicle.Pitch=0
   Fort.Vehicle.RPM=0
   Fort.Vehicle.Throttle=0
   UE.Source.OnPlay=null
   UE.Source.Orientation.Azimuth=0
   UE.Source.Orientation.Elevation=0
   Boost.End.WaveAsset []
  }
  nodes [
   InitVariable.Float()
   InitVariable.Float()
   InitVariable.Float()
   DecibelstoLinearGain_0 = UE.Decibels to Linear Gain:Float(Decibels=^Boost.End.Volume)
   n1 = Array.Shuffle:WaveAsset:Array(In Array=^Boost.End.WaveAsset, Next=^UE.Source.OnPlay)
   n2 = InitVariable.Float()
   n3 = TriggerCompare.Clamp:Bool(A=^Boost.PitchByNormalSpeed, Compare=^UE.Source.OnPlay) {also {defaults {B="<MetaSoundBuilderResult.SUCCEEDED: 0>"}}}
   n4 = UE.Multiply:Float(
    PrimaryOperand=MapRange.MapRange:Float(Out Range B=^Boost.Pitch.Max, Out Range A=^Boost.Pitch.Min, In=/VehicleAudio/Patches/EaseInQuart(Input=^Fort.Vehicle.NormalSpeed).Output).Out Value
    AdditionalOperands=Value.Value:Float(Reset=@n3.False, Set=@n3.True) {also {defaults {Init Value=0, Target Value=1}}}.Output Value
   )
   WavePlayer_0 = UE.Wave Player:Stereo(
    Play=@n1.On Next
    Wave Asset=@n1.Value
    Pitch Shift=VariableAccessor.Float(Variable=VariableMutator.Float(Variable=@n2.Variable, Value=@n4.Out).Variable).Value
   ) {also {version [1,1]}}
   StereoMixer_One_0 = /FortMetasound/Patches/StereoMixer_One("Volume(Lin)"=@DecibelstoLinearGain_0.Linear Gain, InL=@WavePlayer_0.Out Left, InR=@WavePlayer_0.Out Right)
  ]
  links [
   {dst="@UE.OutputFormat.Stereo.Audio:0", src=StereoMixer_One_0.OutL}
   {dst="@UE.OutputFormat.Stereo.Audio:1", src=StereoMixer_One_0.OutR}
   {dst="@UE.Source.OneShot.OnFinished", src=WavePlayer_0.On Finished}
  ]
 }]
properties {
 assetImportData=&self:AssetImportData
 attenuationSettings=/VehicleAudio/Boosts/Vehicle_Boost_Att
 audioPropertiesSheet=/VehicleAudio/PropertySheets/APS_Cosmetic_Vehicle_Boost_End
 maxDistance=5012
 numChannels=2
 outputFormat=Stereo
 sampleRate=48000
 soundClassObject=/Game/Sounds/FortSoundClassesAndModes/Fort_Vehicle_Exterior_Boost
 soundSubmixObject=&Fortnite_SFX
 concurrencySet [/VehicleAudio/Boosts/Vehicle_Boost_End_Conc]
 modulationSettings {
  highpassRouting=Union
  lowpassRouting=Union
  volumeRouting=Union
  highpassModulationDestination {bEnableModulation=true, modulators [/Game/Sounds/Modulation/Patches/HPF/PP_FortVehicles_Boost_HPF]}
  lowpassModulationDestination {bEnableModulation=true, modulators [/Game/Sounds/Modulation/Patches/LPF/PP_FortVehicles_Boost_LPF]}
  volumeModulationDestination {
   bEnableModulation=true
   modulators [/VehicleAudio/Modulation/DM_VehicleTurboPatch, /Game/Sounds/Modulation/Patches/PP_FortVehicles_Boost_Vol]
  }
 }
}
