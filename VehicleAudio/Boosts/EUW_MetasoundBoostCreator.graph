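# Editor utility widget that generates vehicle boost audio assets from selected SoundWaves and an entered name.
# It sorts waves into Start, Loop and End by name, then uses MetaSound builders to make presets of parent MetaSounds, setting wave inputs and speed-pitch toggle.
# Built presets are assigned to a new AGB_Vehicle_Boost child Blueprint's defaults in the current Content Browser folder.
# Other buttons handle folder creation, navigation and standard-boost previews.
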EditorUtilityWidgetBlueprint /VehicleAudio/Boosts/EUW_MetasoundBoostCreator
graphs [
 EventGraph {nodes [
   n1 = GetSelectedAssetsOfClass(AssetClass=/VehicleCosmetics/FX/Blueprints/ParentAssets/Boost/Boost_Parent.Boost_Parent_C)
   > n2 = For Each Loop(Array:Array of Object References=@n1) -> (Array Element:Object Reference)
   > Branch(Condition="Equal Exactly (String)"(A=GetObjectName(Object=@n2), B=BoostCosmetic_AW))
   > Show Message Dialog(Title=Get Title() -> (Title:Text), Message=Get Message() -> (Message:Text), MessageType=Ok, DefaultValue=YesAll, MessageCategory=Warning)
   "On Pressed (EditorUtilityButton_56_MakeFolder)"() {is=title K2Node_ComponentBoundEvent}
   > n3 = Branch(Condition="Equal Exactly (String)"(A=Get SanitizedAssetName() -> (SanitizedAssetName:String)))
   n4 = GetCurrentContentBrowserPath!() -> (OutPath:String, ReturnValue:bool)
   n5 = Append!(A="To String (Text)"(InText="To Text (String)"(InString=@n4)), B=/, C:String=Append!(A=Boost_, B=Get SanitizedAssetName()) -> (ReturnValue)) -> (ReturnValue)
   MakeDirectory!(
    execute=@n3.else
    self:Editor Asset Subsystem Object Reference=Get EditorAssetSubsystem() -> (ReturnValue:Editor Asset Subsystem Object Reference) {is=title K2Node_GetEditorSubsystem}
    DirectoryPath:String=@n5
   ) -> (then, ReturnValue:bool)
   > n6 = Set FolderDirectory(FolderDirectory:String=@n5) -> (Output_Get:String)
   > SyncBrowserToFolders!(execute, FolderList:Array of Strings=Make Array("[0]":String=@n6) -> (Array:Array of Strings)) -> (then)
   "On Pressed (EditorUtilityButton_56_MakeBoost)"() {is=title K2Node_ComponentBoundEvent}
   n7 = Get Text(self=Get AssetNameInput() -> (AssetNameInput:Editor Utility Editable Text Box Object Reference))
   > n8 = Branch(Condition="Equal (Boolean)"(A:bool=TextIsEmpty(InText=@n7), B:bool) -> (ReturnValue:bool=~) {is=title K2Node_PromotableOperator})
   > n9 = GetSelectedAssetsOfClass(AssetClass=/Script/Engine.SoundWave)
   > n10 = Branch(Condition=Is Not Empty!(TargetArray:Array of Object References=@n9) -> (ReturnValue:bool))
   > n11 = Sequence!(execute) -> (then_0, then_1, then_2:Exec, then_3:Exec, then_4:Exec, then_5:Exec, then_6:Exec, then_7:Exec)
   > n12 = MetasoundBuilderInit!(execute, self, Object:Object Reference=Get Boost Start MSS() -> (Boost Start MSS:Meta Sound Source Object Reference)) -> (then, Source Preset Builder:Meta Sound Source Builder Object Reference, BuilderName:Name=None)
   > Set StartMetasoundBuilder(StartMetasoundBuilder:Meta Sound Source Builder Object Reference=@n12) -> (Output_Get:Meta Sound Source Builder Object Reference)
   n13 = MetasoundBuilderInit(execute=@n11.then_1, Object=Get Boost Loop MSS() -> (Boost Loop MSS:Meta Sound Source Object Reference)) -> (BuilderName=None)
   > Set LoopMetasoundBuilder(LoopMetasoundBuilder:Meta Sound Source Builder Object Reference=@n13) -> (Output_Get:Meta Sound Source Builder Object Reference)
   n14 = MetasoundBuilderInit(execute=@n11.then_2, Object=Get Boost End MSS() -> (Boost End MSS:Meta Sound Source Object Reference)) -> (BuilderName=None)
   > Set EndMetasoundBuilder(EndMetasoundBuilder:Meta Sound Source Builder Object Reference=@n14) -> (Output_Get:Meta Sound Source Builder Object Reference)
   n15 = GetSelectedAssetsOfClass(execute=@n11.then_3, AssetClass=&SoundWave)
   > n16 = For Each Loop(Array=@n15)
   n17 = Get BoostStart() -> (BoostStart:Array of Object References)
   Target Builder!(
    execute=@n16.Completed
    self
    WaveAssetArrayParameter:Array of Object References=@n17
    MetasoundInputName:String=Boost.Start.Waveasset
    TargetMetasoundBuilder:Meta Sound Builder Base Object Reference=Get StartMetasoundBuilder() -> (StartMetasoundBuilder:Meta Sound Source Builder Object Reference)
   ) -> (then, NewParam1:bool)
   n18 = Get Boost Loop() -> (Boost Loop:Array of Object References)
   > Target Builder(
    WaveAssetArrayParameter=@n18
    MetasoundInputName=Boost.Loop.Waveasset
    TargetMetasoundBuilder=Get LoopMetasoundBuilder() -> (LoopMetasoundBuilder:Meta Sound Source Builder Object Reference)
   )
   n19 = Get BoostEnd() -> (BoostEnd:Array of Object References)
   > Target Builder(
    WaveAssetArrayParameter=@n19
    MetasoundInputName=Boost.End.Waveasset
    TargetMetasoundBuilder=Get EndMetasoundBuilder() -> (EndMetasoundBuilder:Meta Sound Source Builder Object Reference)
   )
   Sort Wave Assets!(
    execute=@n16.LoopBody
    self
    WaveAsset:Object Reference=@n16
    WaveAssetArrayParameter:Array of Object References=@n17
    Name:Array of Strings=Get Name_1() -> (Name_1:Array of Strings)
   ) -> (then)
   > Sort Wave Assets(WaveAsset=@n16, WaveAssetArrayParameter=@n18, Name=Get Name_0() -> (Name_0:Array of Strings))
   > Sort Wave Assets(WaveAsset=@n16, WaveAssetArrayParameter=@n19, Name=Get Name() -> (Name:Array of Strings))
   n20 = Create MetaSound Literal from Boolean!(
    execute=@n11.then_4
    Value:bool=IsChecked(self=Get EditorUtilityCheckBox_0() -> (EditorUtilityCheckBox_0:Editor Utility Check Box Object Reference))
   ) -> (then, ReturnValue:Meta Sound Literal Structure)
   > n21 = SetGraphInputDefault!(
    execute
    self:Meta Sound Builder Base Object Reference=Get StartMetasoundBuilder()
    InputName:Name=Boost.PitchByNormalSpeed
    Literal:"Meta Sound Literal Structure (by ref)"=@n20
    PageName:Name
   ) -> (Succeeded:Exec, Failed:Exec)
   n22 = SetGraphInputDefault(execute=[@n21.Failed, @n21.Succeeded], self=Get LoopMetasoundBuilder(), InputName=Boost.PitchByNormalSpeed, Literal=@n20)
   SetGraphInputDefault(execute=[@n22.Failed, @n22.Succeeded], self=Get EndMetasoundBuilder(), InputName=Boost.PitchByNormalSpeed, Literal=@n20)
   n23 = Sequence!(execute=@n11.then_5) -> (then_0, then_1, then_2, then_3, then_4, then_5)
   n24 = Get BoostStart()
   > Branch(Condition="Not Equal (Integer)"(A=Length(TargetArray:Array of Object References=@n24)))
   n25 = GetPlatformUserName!() -> (ReturnValue:String)
   > n26 = BuildToAsset!(
    execute
    self:Meta Sound Editor Subsystem Object Reference=Get MetaSoundEditorSubsystem() -> (ReturnValue:Meta Sound Editor Subsystem Object Reference) {is=title K2Node_GetEditorSubsystem}
    InBuilder:Meta Sound Builder Base Object Reference=Get StartMetasoundBuilder()
    Author:String=@n25
    AssetName:String=Append!(A=MSS_, B=Vehicle_, C:String=Boost_, D:String=Get SanitizedAssetName(), E:String=_Start, F:String=_Preset) -> (ReturnValue)
    PackagePath:String=GetCurrentContentBrowserPath()
    TemplateSoundWave:Sound Wave Object Reference=/VehicleAudio/Boosts/_ParentMetasound/MSS_Vehicle_Boost_Start
   ) -> (Succeeded:Exec, Failed:Exec, ReturnValue:Meta Sound Document Interface Interface)
   > n27 = Cast To MetaSoundSource!(execute, Object:Meta Sound Document Interface Interface=@n26) -> (then, CastFailed, AsMetaSound Source:Meta Sound Source Object Reference) {is=title K2Node_DynamicCast}
   > Set StartSoundAsset(StartSoundAsset:Meta Sound Source Object Reference=@n27) -> (Output_Get:Meta Sound Source Object Reference)
   n28 = Get Boost Loop()
   Branch(execute=@n23.then_1, Condition="Not Equal (Integer)"(A=Length(TargetArray=@n28)))
   > n29 = BuildToAsset(
    self=Get MetaSoundEditorSubsystem()
    InBuilder=Get LoopMetasoundBuilder()
    Author=@n25
    AssetName=Append(A=MSS_, B=Vehicle_, C=Boost_, D=Get SanitizedAssetName(), E=_Loop, F=_Preset)
    PackagePath=GetCurrentContentBrowserPath()
    TemplateSoundWave=/VehicleAudio/Boosts/_ParentMetasound/MSS_Vehicle_Boost_Loop
   )
   > n30 = Cast To MetaSoundSource(Object=@n29)
   > Set LoopSoundAsset(LoopSoundAsset:Meta Sound Source Object Reference=@n30) -> (Output_Get:Meta Sound Source Object Reference)
   n31 = Get BoostEnd()
   Branch(execute=@n23.then_2, Condition="Not Equal (Integer)"(A=Length(TargetArray=@n31)))
   > n32 = BuildToAsset(
    self=Get MetaSoundEditorSubsystem()
    InBuilder=Get EndMetasoundBuilder()
    Author=@n25
    AssetName=Append(A=MSS_, B=Vehicle_, C=Boost_, D=Get SanitizedAssetName(), E=_End, F=_Preset)
    PackagePath=GetCurrentContentBrowserPath()
    TemplateSoundWave=/VehicleAudio/Boosts/_ParentMetasound/MSS_Vehicle_Boost_End
   )
   > n33 = Cast To MetaSoundSource(Object=@n32)
   > Set EndSoundAsset(EndSoundAsset:Meta Sound Source Object Reference=@n33) -> (Output_Get:Meta Sound Source Object Reference)
   Clear(execute=@n23.then_3, TargetArray:Array of Object References=Get Boost Loop())
   > Clear(TargetArray=Get BoostEnd())
   > Clear(TargetArray=Get BoostStart())
   n34 = Create New Blueprint Component!(
    execute=@n23.then_4
    ComponentClass:Actor Component Class Reference=&/AGB_Vehicle_Boost.AGB_Vehicle_Boost_C
    NewClassPath:String=GetCurrentContentBrowserPath()
    NewClassName:String=Append!(A=AGB_Vehicle_Boost_, B=Get SanitizedAssetName()) -> (ReturnValue)
    None:%=Get SubobjectDataSubsystem()
   ) -> (then, ReturnValue:Object Class Reference)
   > n35 = GetClassDefaultObject!(execute, Class:Object Class Reference=@n34) -> (then, ReturnValue:Object Reference)
   > n36 = Cast To AGB_Vehicle_Boost!(execute, Object=@n35) -> (then, CastFailed, AsAGB Vehicle Boost:AGB Vehicle Boost Object Reference) {is=title K2Node_DynamicCast}
   > n37 = Sequence!(execute) -> (then_0, then_1, then_2)
   n38 = Get LoopSoundAsset() -> (LoopSoundAsset:Meta Sound Source Object Reference)
   > Is Valid(InputObject=@n38)
   > Set LoopingSound!(execute, LoopingSound:Sound Base Object Reference=@n38, self:AGB Vehicle Boost Object Reference=@n36) -> (then, Output_Get:Sound Base Object Reference)
   n39 = Get EndSoundAsset() -> (EndSoundAsset:Meta Sound Source Object Reference)
   Is Valid(exec=@n37.then_1, InputObject=@n39)
   > Set EndSound!(execute, EndSound:Sound Base Object Reference=@n39, self:AGB Vehicle Boost Object Reference=@n36) -> (then, Output_Get:Sound Base Object Reference)
   n40 = Get StartSoundAsset() -> (StartSoundAsset:Meta Sound Source Object Reference)
   Is Valid(exec=@n37.then_2, InputObject=@n40)
   > Set StartSound!(execute, StartSound:Sound Base Object Reference=@n40, self:AGB Vehicle Boost Object Reference=@n36) -> (then, Output_Get:Sound Base Object Reference)
   n41 = GetCurrentContentBrowserPath()
   SaveDirectory!(execute=@n23.then_5, DirectoryPath:String=@n41, bOnlyIfIsDirty:bool=false, bRecursive:bool=false) -> (then, ReturnValue:bool)
   > SaveDirectory(DirectoryPath=@n41, bOnlyIfIsDirty=false, bRecursive=false)
   Show Message Dialog(
    execute=@n10.else
    Title="To Text (String)"(InString=Error:)
    Message="To Text (String)"(InString="No sound waves selected. Try selecting soundwaves, and then click \"Get Current Path\"")
    MessageType=Ok
    DefaultValue=No
    MessageCategory=Error
   )
   Show Message Dialog(
    execute=@n8.else
    Title="To Text (String)"(InString=Error:)
    Message="To Text (String)"(InString="No asset name input selected. Please input an asset name. ")
    MessageType=Ok
    DefaultValue=No
    MessageCategory=Error
   )
   "On Pressed (EditorUtilityButton)"() {is=title K2Node_ComponentBoundEvent}
   > Set Text(
    Text="To Text (String)"(InString=GetCurrentContentBrowserPath())
    self:Text Box Object Reference=Get PathInput() -> (PathInput:Editor Utility Editable Text Box Object Reference)
   )
   "On Pressed (EditorUtilityButton_BoostStandardStart)"() {is=title K2Node_ComponentBoundEvent}
   > PlaySound2D!(
    execute
    WorldContextObject:Object Reference
    Sound=/VehicleAudio/Boosts/Boost_Standard/MSS_Vehicle_Boost_Standard_Start_Preset
    VolumeMultiplier
    PitchMultiplier
    StartTime
    ConcurrencySettings
    OwningActor
    bIsUISound
   ) -> (then)
   "On Pressed (EditorUtilityButton_1_playselectedsounds)"() {is=title K2Node_ComponentBoundEvent}
   > n42 = GetSelectedAssetsOfClass(AssetClass=/Script/MetasoundEngine.MetaSoundSource)
   > n43 = For Each Loop(Array=@n42)
   > n44 = Cast To SoundWave!(execute, Object=@n43) -> (then, CastFailed, AsSound Wave:Sound Wave Object Reference) {is=title K2Node_DynamicCast}
   > PlaySound2D(Sound=@n44)
   "On Pressed (EditorUtilityButton_BoostStandardLoop)"() {is=title K2Node_ComponentBoundEvent}
   > n45 = SpawnSound2D!(
    execute
    WorldContextObject:Object Reference
    Sound:Sound Base Object Reference=/VehicleAudio/Boosts/Boost_Standard/MSS_Vehicle_Boost_Standard_Loop_Preset
    VolumeMultiplier:float=1
    PitchMultiplier:float=1
    StartTime:float=0
    ConcurrencySettings:Sound Concurrency Object Reference
    bPersistAcrossLevelTransition:bool=false
    bAutoDestroy:bool=true
   ) -> (then, ReturnValue:Audio Component Object Reference)
   > Set BoostStandardLoopPreview(BoostStandardLoopPreview:Audio Component Object Reference=@n45) -> (Output_Get:Audio Component Object Reference)
   n46 = "On Text Changed (AssetNameInput)"!() -> (OutputDelegate, then, Text:"Text (by ref)") {is=title K2Node_ComponentBoundEvent}
   n47 = "To String (Text)"(InText=@n46.Text)
   Set SanitizedAssetName(
    execute=[@n46.then, @n51.then]
    SanitizedAssetName:String=TrimTrailing!(SourceString:String=Trim!(SourceString:String=@n47) -> (ReturnValue:String)) -> (ReturnValue:String)
   ) -> (Output_Get:String)
   "On Released (EditorUtilityButton_BoostStandardLoop)"() {is=title K2Node_ComponentBoundEvent}
   > FadeOut!(
    execute
    self:Audio Component Object Reference=Get BoostStandardLoopPreview() -> (BoostStandardLoopPreview:Audio Component Object Reference)
    FadeOutDuration:float=0.3
    FadeVolumeLevel:float=0
    FadeCurve:EAudioFaderCurve Enum=Linear
   ) -> (then)
   > Set BoostStandardLoopPreview()
   "On Pressed (EditorUtilityButton_BoostDirectory)"() {is=title K2Node_ComponentBoundEvent}
   > SyncBrowserToFolders(FolderList=Make Array("[0]"=/VehicleAudio/Boosts))
   "On Pressed (EditorUtilityButton_BoostDVisualirectory)"() {is=title K2Node_ComponentBoundEvent}
   n48 = Append!(A=/VehicleCosmetics/Boosts/, B=Boost_, C:String=Get SanitizedAssetName()) -> (ReturnValue)
   > n49 = DoesDirectoryExist!(execute, DirectoryPath:String=@n48) -> (then, ReturnValue:bool)
   > n50 = Branch(Condition=@n49)
   > SyncBrowserToFolders(FolderList=Make Array("[0]"=@n48))
   SyncBrowserToFolders(execute=@n50.else, FolderList=Make Array("[0]"=&Boosts))
   n51 = Event Tick()
   n52 = Sequence!(execute) -> (then_0, then_1)
   > n53 = GetSelectedAssetsOfClass(AssetClass=/Script/Engine.SoundBase)
   > n54 = For Each Loop(Array=@n53)
   > n55 = Branch(Condition=Contains(SearchIn=GetObjectName(Object=@n54), Substring=Loop))
   > n56 = Cast To SoundBase!(execute, Object=@n54) -> (then, CastFailed, AsSound Base:Sound Base Object Reference) {is=title K2Node_DynamicCast}
   > n57 = Set Looping Sound(Looping Sound:Sound Base Object Reference=@n56) -> (Output_Get:Sound Base Object Reference)
   n58 = Branch(execute=[@n55.else, @n57.then], Condition=Contains(SearchIn=GetObjectName(Object=@n54), Substring=Start))
   > n59 = Cast To SoundBase(Object=@n54)
   > n60 = Set Start Sound(Start Sound:Sound Base Object Reference=@n59) -> (Output_Get:Sound Base Object Reference)
   Branch(execute=[@n58.else, @n60.then], Condition=Contains(SearchIn=GetObjectName(Object=@n54), Substring=End))
   > n61 = Cast To SoundBase(Object=@n54)
   > Set End Sound(End Sound:Sound Base Object Reference=@n61) -> (Output_Get:Sound Base Object Reference)
   n62 = GetSelectedAssetsOfClass(execute=@n52.then_1, AssetClass=&AGB_Vehicle_Boost_C)
   > n63 = For Each Loop(Array=@n62)
   n64 = Cast To AGB_Vehicle_Boost(execute=[@n63.Completed, @n63.LoopBody], Object=@n63)
   > Set LoopingSound(LoopingSound=Get Looping Sound() -> (Looping Sound:Sound Base Object Reference), self=@n64)
   > Set EndSound(EndSound=Get End Sound() -> (End Sound:Sound Base Object Reference), self=@n64)
   > Set StartSound(StartSound=Get Start Sound() -> (Start Sound:Sound Base Object Reference), self=@n64)
   LogString(execute=@n64.CastFailed, InString=Hello)
   Contains Item(TargetArray:Array of Wildcards, ItemToFind:"Wildcard (by ref)")
   "To String (Text)"(InText=Get Text(self=Get PathInput()))
   "To String (Text)"(InText=Get Text(self=Get PathInput()))
   "To String (Text)"(InText=Get Text(self=Get PathInput()))
   String To Name(InString=Append!(A=BoostCosmetic_, B=Get SanitizedAssetName()) -> (ReturnValue))
   ClassIsChildOf(
    TestClass=GetClass!(InAssetData:"Asset Data Structure (by ref)") -> (ReturnValue:Object Class Reference)
    ParentClass=&Boost_Parent_C
   )
   SetFilterTagsAndValues!(
    InFilter:"ARFilter Structure (by ref)"=Get Filter() -> (Filter:ARFilter Structure)
    InTagsAndValues:Array of Tag And Value Structures=Get In Tags and Values() -> (In Tags and Values:Array of Tag And Value Structures)
   ) -> (ReturnValue:ARFilter Structure)
   SyncBrowserToObjects(AssetPaths=Make Array("[0]"=Get Object Path String(Object=@n36)))
   "On Pressed (EditorUtilityButton_56_MakeBoost_1)"() {is=title K2Node_ComponentBoundEvent}
   Event Construct()
   Event PreConstruct()
   Get AssetEditorSubsystem()
   Branch(Condition="Equal (Integer)"(A=Length(TargetArray:Array of Wildcards), B=1))
   Make Array("[0]":Object Reference=@n35) -> (Array:Array of Object References)
   Get Text(self=Get AssetNameInput())
   Get SanitizedAssetName()
   Set SanitizedAssetName()
  ]}
 MetasoundBuilderInit {nodes [
   n1 = MetasoundBuilderInit!() -> (then, Object:Object Reference) {is=title K2Node_FunctionEntry}
   > n2 = Cast To MetaSoundDocumentInterface!(execute, Object=@n1) -> (then, CastFailed, AsMetaSound Document Interface:Meta Sound Document Interface Interface) {is=title K2Node_DynamicCast}
   n3 = "To Text (Object)"!(InObj:Object Reference=@n1) -> (ReturnValue:Text)
   n4 = String To Name(InString=Append(A=Builder_, B="To String (Text)"(InText=@n3)))
   > n5 = CreateSourcePresetBuilder!(
    execute
    self:Meta Sound Builder Subsystem Object Reference=Get MetaSoundBuilderSubsystem() -> (ReturnValue:Meta Sound Builder Subsystem Object Reference) {is=title K2Node_GetEngineSubsystem}
    BuilderName:Name=@n4
    ReferencedSourceClass:"Meta Sound Document Interface Interface (by ref)"=@n2
   ) -> (Succeeded:Exec, Failed:Exec, ReturnValue:Meta Sound Source Builder Object Reference)
   > Return Node!(execute, Source Preset Builder:Meta Sound Source Builder Object Reference=@n5, BuilderName:Name=@n4)
   PrintString(execute=@n2.CastFailed, InString="Your cast failed! Maybe something is wrong with the boost metasounds")
  ]}
 Sort Wave Assets {nodes [
   n1 = Sort Wave Assets!() -> (then, WaveAsset:Object Reference, WaveAssetArrayParameter:Array of Object References, Name:Array of Strings) {is=title K2Node_FunctionEntry}
   > n2 = For Each Loop(Array:Array of Strings=@n1.Name) -> (Array Element:String)
   > Branch(Condition=Contains(SearchIn=GetObjectName(Object=@n1), Substring=@n2))
   > Add Unique(TargetArray:Array of Object References=@n1.WaveAssetArrayParameter, NewItem:"Object Reference (by ref)"=@n1)
  ]}
 Target Builder {nodes [
   n1 = Target Builder!() -> (
    then
    WaveAssetArrayParameter:Array of Object References
    MetasoundInputName:String
    TargetMetasoundBuilder:Meta Sound Builder Base Object Reference
   ) {is=title K2Node_FunctionEntry}
   > n2 = Branch(Condition="integer > integer"(A=Length(TargetArray:Array of Object References=@n1)))
   > n3 = Create MetaSound Literal From AudioParameter!(
    execute
    self:Meta Sound Builder Subsystem Object Reference=Get MetaSoundBuilderSubsystem()
    Param:"Audio Parameter Structure (by ref)"=ObjectArrayToAudioParameter!(Name:Name=String To Name(InString=@n1.MetasoundInputName), Objects:Array of Object References=@n1) -> (ReturnValue:Audio Parameter Structure)
   ) -> (then, ReturnValue:Meta Sound Literal Structure)
   > n4 = SetGraphInputDefault(self=@n1.TargetMetasoundBuilder, InputName=String To Name(InString=@n1.MetasoundInputName), Literal=@n3)
   Return Node!(execute=@n4.Failed, NewParam1:bool=false)
   > Return Node(NewParam1=true)
   Return Node(execute=@n2.else, NewParam1=false)
  ]}
]
variables [
 As Meta Sound Source {type=MetaSoundSource, default=null, class=~, properties=~}
 Assets {type="array<Object>", default []}
 AudioGameplayBehavior {type=AudioGameplayBehaviorInstance, default {sound=&AGB_Vehicle_Boost_C}}
 Boost End MSS {type=MetaSoundSource, default=&MSS_Vehicle_Boost_End, class=~, properties=~}
 Boost End Soundwave {type="array<SoundWave>", default []}
 Boost Loop {type="array<Object>", default []}
 Boost Loop MSS {type=MetaSoundSource, default=&MSS_Vehicle_Boost_Loop, class=~, properties=~}
 Boost Loop Soundwave {type="array<SoundWave>", default []}
 Boost Start MSS {type=MetaSoundSource, default=&MSS_Vehicle_Boost_Start, class=~, properties=~}
 Boost Start Soundwave {type="array<SoundWave>", default []}
 BoostAGB {type=Object, default=null}
 BoostEnd {type="array<Object>", default []}
 BoostStandardLoopPreview {type=AudioComponent, default=null}
 BoostStart {type="array<Object>", default []}
 End Sound {type=SoundBase, default=null}
 EndMetasoundBuilder {type=MetaSoundSourceBuilder, default=null}
 EndSoundAsset {type=MetaSoundSource, default=null, class=~, properties=~}
 Filter {
  type=ARFilter
  default {
   bIncludeOnlyOnDiskAssets=false
   bRecursiveClasses=false
   bRecursivePaths=false
   classNames []
   classPaths []
   objectPaths []
   packageNames []
   packagePaths [BoostCosmetic_AW]
   recursiveClassPathsExclusionSet []
   recursiveClassesExclusionSet []
   softObjectPaths []
  }
 }
 Folder List {type="array<string>", default []}
 FolderDirectory {type=string, default=""}
 In Filter {
  type=ARFilter
  default {
   bIncludeOnlyOnDiskAssets=false
   bRecursiveClasses=false
   bRecursivePaths=false
   classNames []
   classPaths []
   objectPaths []
   packageNames []
   packagePaths []
   recursiveClassPathsExclusionSet []
   recursiveClassesExclusionSet []
   softObjectPaths []
  }
 }
 In Tags and Values {type="array<TagAndValue>", default [{}, {}]}
 Looping Sound {type=SoundBase, default=null}
 LoopMetasoundBuilder {type=MetaSoundSourceBuilder, default=null}
 LoopSoundAsset {type=MetaSoundSource, default=null, class=~, properties=~}
 Message {type=string, default=yes}
 Metasound Builder {type=MetaSoundSourceBuilder, default=null}
 MetasoundBoostEnd {default=None}
 MetasoundBoostLoop {default=None}
 MetasoundBoostStart {default=None}
 MetasoundEndReference {type=Object, default=null}
 MetasoundLoopReference {type=Object, default=null}
 MetasoundStartReference {type=Object, default=null}
 Name {type="array<string>", default [Stop,End]}
 Name_0 {type="array<string>", default [Loop]}
 Name_1 {type="array<string>", default [Start]}
 NewVar {type=AudioComponent, default=null}
 SanitizedAssetName {type=string, default=""}
 Start Sound {type=SoundBase, default=null}
 StartMetasoundBuilder {type=MetaSoundSourceBuilder, default=null}
 StartSoundAsset {type=MetaSoundSource, default=null, class=~, properties=~}
 Title {type=string, default=meesage}
]
widgets {
 info {rootWidgetClass=&CanvasPanel, widgetCount=30}
 widgets [
  CanvasPanel_202 {widgetClassPath=&CanvasPanel, widget=CanvasPanel_202}
  EditorUtilityButton_56_MakeBoost {
   widgetClassPath=&EditorUtilityButton
   parent=CanvasPanel_202
   widget=EditorUtilityButton_56_MakeBoost
   properties {toolTipText=Make metasounds from selected soundwaves}
   slot {
    class=CanvasPanelSlot
    properties {layoutData {alignment=<0 0>, anchors {maximum=<0 0>, minimum=<0 0>}, offsets=<ltrb 372 300 272.6727 51.7718>}}
   }
  }
  UEFN_TextBlock_C_39 {
   widgetClassPath=&UEFN_TextBlock
   parent=EditorUtilityButton_56_MakeBoost
   widget=UEFN_TextBlock_C_39
   properties {
    justification=Center
    text=2. Make Boost Objects
    font {
     bForceMonospaced=false
     bMaterialIsStencil=false
     fontMaterial=null
     fontObject=&BurbankSmall
     letterSpacing=0
     monospacedWidth=1
     size=9
     skewAmount=0
     typefaceFontName=Black
     outlineSettings {
      bApplyOutlineToDropShadows=false
      bMiteredCorners=false
      bSeparateFillAlpha=false
      outlineMaterial=null
      outlineSize=0
      outlineColor=<rgba 0 0 0 1>
     }
    }
   }
   slot {class=ButtonSlot}
  }
  EditorUtilityButton_1_playselectedsounds {
   widgetClassPath=&EditorUtilityButton
   parent=CanvasPanel_202
   widget=EditorUtilityButton_1_playselectedsounds
   properties {visibility=Hidden}
   slot {
    class=CanvasPanelSlot
    properties {layoutData {alignment=<0 0>, anchors {maximum=<0 0>, minimum=<0 0>}, offsets=<ltrb 12 944 272.6727 51.7718>}}
   }
  }
  UEFN_TextBlock_C_3 {
   widgetClassPath=&UEFN_TextBlock
   parent=EditorUtilityButton_1_playselectedsounds
   widget=UEFN_TextBlock_C_3
   properties {
    justification=Center
    text=Play Selected Sounds
    font {
     bForceMonospaced=false
     bMaterialIsStencil=false
     fontMaterial=null
     fontObject=&BurbankSmall
     letterSpacing=0
     monospacedWidth=1
     size=9
     skewAmount=0
     typefaceFontName=Black
     outlineSettings {
      bApplyOutlineToDropShadows=false
      bMiteredCorners=false
      bSeparateFillAlpha=false
      outlineMaterial=null
      outlineSize=0
      outlineColor=<rgba 0 0 0 1>
     }
    }
   }
   slot {class=ButtonSlot}
  }
  PathInput {
   widgetClassPath=&EditorUtilityEditableTextBox
   bIsVariable=true
   parent=CanvasPanel_202
   widget=PathInput
   slot {
    class=CanvasPanelSlot
    properties {layoutData {alignment=<0 0>, anchors {maximum=<0 0>, minimum=<0 0>}, offsets=<ltrb 152 136 313.964 28.4985>}}
   }
  }
  UEFN_TextBlock_C {
   widgetClassPath=&UEFN_TextBlock
   parent=CanvasPanel_202
   widget=UEFN_TextBlock_C
   properties {
    text=Path
    font {
     bForceMonospaced=false
     bMaterialIsStencil=false
     fontMaterial=null
     fontObject=&Roboto
     letterSpacing=0
     monospacedWidth=1
     size=10.5
     skewAmount=0
     typefaceFontName=Regular
     outlineSettings {
      bApplyOutlineToDropShadows=false
      bMiteredCorners=false
      bSeparateFillAlpha=false
      outlineMaterial=null
      outlineSize=0
      outlineColor=<rgba 0 0 0 1>
     }
    }
   }
   slot {
    class=CanvasPanelSlot
    properties {layoutData {alignment=<0 0>, anchors {maximum=<0 0>, minimum=<0 0>}, offsets=<ltrb 4 136 101.1862 19.2282>}}
   }
  }
  EditorUtilityButton {
   widgetClassPath=&EditorUtilityButton
   parent=CanvasPanel_202
   widget=EditorUtilityButton
   properties {toolTipText=Set current editor path as the preset build destination}
   slot {
    class=CanvasPanelSlot
    properties {layoutData {alignment=<0 0>, anchors {maximum=<0 0>, minimum=<0 0>}, offsets=<ltrb 480 108 200.9009 72.6426>}}
   }
  }
  UEFN_TextBlock_C_1 {
   widgetClassPath=&UEFN_TextBlock
   parent=EditorUtilityButton
   widget=UEFN_TextBlock_C_1
   properties {
    justification=Center
    text=Get Current Path
    font {
     bForceMonospaced=false
     bMaterialIsStencil=false
     fontMaterial=null
     fontObject=&Roboto
     letterSpacing=0
     monospacedWidth=1
     size=9
     skewAmount=0
     typefaceFontName=Regular
     outlineSettings {
      bApplyOutlineToDropShadows=false
      bMiteredCorners=false
      bSeparateFillAlpha=false
      outlineMaterial=null
      outlineSize=0
      outlineColor=<rgba 0 0 0 1>
     }
    }
   }
   slot {class=ButtonSlot}
  }
  EditorUtilityButton_BoostDirectory {
   widgetClassPath=&EditorUtilityButton
   parent=CanvasPanel_202
   widget=EditorUtilityButton_BoostDirectory
   properties {
    toolTipText=Set current editor path as the preset build destination
    backgroundColor=<rgba 1 0.1725 0 1>
    widgetStyle {
     clickedSlateSound {resourceObject=null}
     disabled {
      bIsDynamicallyLoaded=false
      drawAs=RoundedBox
      imageType=NoImage
      mirroring=NoMirror
      resourceName=None
      resourceObject=null
      tiling=NoTile
      imageSize=<0 0>
      margin=<ltrb 0 0 0 0>
      outlineSettings {
       bUseBrushTransparency=false
       roundingType=FixedRadius
       width=1
       color {colorUseRule=UseColor_ColorTable, specifiedColor=<rgba 0 0 0 0>}
       cornerRadii=<4 4 4 4>
      }
      tintColor {colorUseRule=UseColor_Specified, specifiedColor=<rgba 0.0395 0.0395 0.0395 1>}
      uVRegion {bIsValid=false, max=<0 0>, min=<0 0>}
     }
     disabledForeground {colorUseRule=UseColor_ColorTable, specifiedColor=<rgba 0 0 0 0>}
     hovered {
      bIsDynamicallyLoaded=false
      drawAs=RoundedBox
      imageType=NoImage
      mirroring=NoMirror
      resourceName=None
      resourceObject=null
      tiling=NoTile
      imageSize=<32 32>
      margin=<ltrb 0 0 0 0>
      outlineSettings {
       bUseBrushTransparency=false
       roundingType=FixedRadius
       width=1
       color {colorUseRule=UseColor_ColorTable, specifiedColor=<rgba 0 0 0 0>}
       cornerRadii=<4 4 4 4>
      }
      tintColor {colorUseRule=UseColor_Specified, specifiedColor=<rgba 0.0953 0.0953 0.0953 1>}
      uVRegion {bIsValid=false, max=<0 0>, min=<0 0>}
     }
     hoveredForeground {colorUseRule=UseColor_ColorTable, specifiedColor=<rgba 0 0 0 0>}
     hoveredSlateSound {resourceObject=null}
     normal {
      bIsDynamicallyLoaded=false
      drawAs=RoundedBox
      imageType=NoImage
      mirroring=NoMirror
      resourceName=None
      resourceObject=null
      tiling=NoTile
      imageSize=<32 32>
      margin=<ltrb 0 0 0 0>
      outlineSettings {
       bUseBrushTransparency=false
       roundingType=FixedRadius
       width=1
       color {colorUseRule=UseColor_ColorTable, specifiedColor=<rgba 0 0 0 0>}
       cornerRadii=<4 4 4 4>
      }
      tintColor {colorUseRule=UseColor_Specified, specifiedColor=<rgba 0.0395 0.0395 0.0395 1>}
      uVRegion {bIsValid=false, max=<0 0>, min=<0 0>}
     }
     normalForeground {colorUseRule=UseColor_Specified, specifiedColor=<rgba 0 0 0 0>}
     normalPadding=<ltrb 12 1.5 12 1.5>
     pressed {
      bIsDynamicallyLoaded=false
      drawAs=RoundedBox
      imageType=NoImage
      mirroring=NoMirror
      resourceName=None
      resourceObject=null
      tiling=NoTile
      imageSize=<32 32>
      margin=<ltrb 0 0 0 0>
      outlineSettings {
       bUseBrushTransparency=false
       roundingType=FixedRadius
       width=1
       color {colorUseRule=UseColor_ColorTable, specifiedColor=<rgba 0 0 0 0>}
       cornerRadii=<4 4 4 4>
      }
      tintColor {colorUseRule=UseColor_Specified, specifiedColor=<rgba 0.0284 0.0284 0.0284 1>}
      uVRegion {bIsValid=false, max=<0 0>, min=<0 0>}
     }
     pressedForeground {colorUseRule=UseColor_ColorTable, specifiedColor=<rgba 0 0 0 0>}
     pressedPadding=<ltrb 12 2.5 12 0.5>
     pressedSlateSound {resourceObject=null}
    }
   }
   slot {
    class=CanvasPanelSlot
    properties {layoutData {alignment=<0 0>, anchors {maximum=<0 0>, minimum=<0 0>}, offsets=<ltrb 480 188 200.9009 72.6426>}}
   }
  }
  UEFN_TextBlock_C_7 {
   widgetClassPath=&UEFN_TextBlock
   parent=EditorUtilityButton_BoostDirectory
   widget=UEFN_TextBlock_C_7
   properties {
    justification=Center
    text=Boost Sound Directory
    font {
     bForceMonospaced=false
     bMaterialIsStencil=false
     fontMaterial=null
     fontObject=&Roboto
     letterSpacing=0
     monospacedWidth=1
     size=9
     skewAmount=0
     typefaceFontName=Regular
     outlineSettings {
      bApplyOutlineToDropShadows=false
      bMiteredCorners=false
      bSeparateFillAlpha=false
      outlineMaterial=null
      outlineSize=0
      outlineColor=<rgba 0 0 0 1>
     }
    }
   }
   slot {class=ButtonSlot}
  }
  UEFN_TextBlock_C_2 {
   widgetClassPath=&UEFN_TextBlock
   parent=CanvasPanel_202
   widget=UEFN_TextBlock_C_2
   properties {
    text=Asset Name
    font {
     bForceMonospaced=false
     bMaterialIsStencil=false
     fontMaterial=null
     fontObject=&Roboto
     letterSpacing=0
     monospacedWidth=1
     size=10.5
     skewAmount=0
     typefaceFontName=Regular
     outlineSettings {
      bApplyOutlineToDropShadows=false
      bMiteredCorners=false
      bSeparateFillAlpha=false
      outlineMaterial=null
      outlineSize=0
      outlineColor=<rgba 0 0 0 1>
     }
    }
   }
   slot {
    class=CanvasPanelSlot
    properties {layoutData {alignment=<0 0>, anchors {maximum=<0 0>, minimum=<0 0>}, offsets=<ltrb 4 216 101.1862 19.2282>}}
   }
  }
  UEFN_TextBlock_C_4 {
   widgetClassPath=&UEFN_TextBlock
   parent=CanvasPanel_202
   widget=UEFN_TextBlock_C_4
   properties {
    text=Should this boost have pitching based on speed?
    font {
     bForceMonospaced=false
     bMaterialIsStencil=false
     fontMaterial=null
     fontObject=&Roboto
     letterSpacing=0
     monospacedWidth=1
     size=10.5
     skewAmount=0
     typefaceFontName=Regular
     outlineSettings {
      bApplyOutlineToDropShadows=false
      bMiteredCorners=false
      bSeparateFillAlpha=false
      outlineMaterial=null
      outlineSize=0
      outlineColor=<rgba 0 0 0 1>
     }
    }
   }
   slot {
    class=CanvasPanelSlot
    properties {layoutData {alignment=<0 0>, anchors {maximum=<0 0>, minimum=<0 0>}, offsets=<ltrb 84 256 101.1862 19.2282>}}
   }
  }
  AssetNameInput {
   widgetClassPath=&EditorUtilityEditableTextBox
   bIsVariable=true
   parent=CanvasPanel_202
   widget=AssetNameInput
   properties {toolTipText=What is the name of he product}
   slot {
    class=CanvasPanelSlot
    properties {layoutData {alignment=<0 0>, anchors {maximum=<0 0>, minimum=<0 0>}, offsets=<ltrb 152 212 313.964 28.4985>}}
   }
  }
  EditorUtilityButton_56_MakeBoost_1 {
   widgetClassPath=&EditorUtilityButton
   parent=CanvasPanel_202
   widget=EditorUtilityButton_56_MakeBoost_1
   properties {toolTipText=Once all wave assets are selected make 1p weapon preset, visibility=Hidden}
   slot {
    class=CanvasPanelSlot
    properties {layoutData {alignment=<0 0>, anchors {maximum=<0 0>, minimum=<0 0>}, offsets=<ltrb 312 944 272.6727 51.7718>}}
   }
  }
  CreateBlueprintButton {
   widgetClassPath=&UEFN_TextBlock
   bIsVariable=true
   parent=EditorUtilityButton_56_MakeBoost_1
   widget=CreateBlueprintButton
   properties {
    justification=Center
    text=Create Blueprint
    font {
     bForceMonospaced=false
     bMaterialIsStencil=false
     fontMaterial=null
     fontObject=&BurbankSmall
     letterSpacing=0
     monospacedWidth=1
     size=9
     skewAmount=0
     typefaceFontName=Black
     outlineSettings {
      bApplyOutlineToDropShadows=false
      bMiteredCorners=false
      bSeparateFillAlpha=false
      outlineMaterial=null
      outlineSize=0
      outlineColor=<rgba 0 0 0 1>
     }
    }
   }
   slot {class=ButtonSlot}
  }
  Image_0 {
   widgetClassPath=&Image
   bIsVariable=true
   parent=CanvasPanel_202
   widget=Image_0
   properties {brush {
     bIsDynamicallyLoaded=false
     drawAs=Image
     imageType=FullColor
     mirroring=NoMirror
     resourceName=None
     resourceObject=/VehicleCosmetics/UI/Icons/Body/S28/Icons/T-UI-Icon-Body-Octane
     tiling=NoTile
     imageSize=<128 128>
     margin=<ltrb 0 0 0 0>
     outlineSettings {
      bUseBrushTransparency=false
      roundingType=HalfHeightRadius
      width=0
      color {colorUseRule=UseColor_Specified, specifiedColor=<rgba 0 0 0 0>}
      cornerRadii=<0 0 0 0>
     }
     tintColor {colorUseRule=UseColor_Specified, specifiedColor=<rgba 1 1 1 1>}
     uVRegion {bIsValid=false, max=<0 0>, min=<0 0>}
    }}
   slot {
    class=CanvasPanelSlot
    properties {layoutData {alignment=<0 0>, anchors {maximum=<0 0>, minimum=<0 0>}, offsets=<ltrb 0 0 274.8415 128.7554>}}
   }
  }
  Image {
   widgetClassPath=&Image
   bIsVariable=true
   parent=CanvasPanel_202
   widget=Image
   properties {brush {
     bIsDynamicallyLoaded=false
     drawAs=Image
     imageType=FullColor
     mirroring=NoMirror
     resourceName=None
     resourceObject=/VehicleCosmetics/UI/Icons/Boosts/T-UI-Icon-Boost-Boost-Standard
     tiling=NoTile
     imageSize=<128 128>
     margin=<ltrb 0 0 0 0>
     outlineSettings {
      bUseBrushTransparency=false
      roundingType=HalfHeightRadius
      width=0
      color {colorUseRule=UseColor_Specified, specifiedColor=<rgba 0 0 0 0>}
      cornerRadii=<0 0 0 0>
     }
     tintColor {colorUseRule=UseColor_Specified, specifiedColor=<rgba 1 1 1 1>}
     uVRegion {bIsValid=false, max=<0 0>, min=<0 0>}
    }}
   slot {
    class=CanvasPanelSlot
    properties {layoutData {alignment=<0 0>, anchors {maximum=<0 0>, minimum=<0 0>}, offsets=<ltrb 256 0 274.8415 128.7554>}}
   }
  }
  EditorUtilityCheckBox_0 {
   widgetClassPath=&EditorUtilityCheckBox
   bIsVariable=true
   parent=CanvasPanel_202
   widget=EditorUtilityCheckBox_0
   properties {
    checkedState=Checked
    toolTipText="by default, most boosts should pitch based on the speed of the thing it's attached to. If it's music, you probably want to uncheck this"
   }
   slot {class=CanvasPanelSlot, properties {layoutData {alignment=<0 0>, anchors {maximum=<0 0>, minimum=<0 0>}, offsets=<ltrb 44 248 53.954 40>}}}
  }
  Image_1 {
   widgetClassPath=&Image
   bIsVariable=true
   parent=CanvasPanel_202
   widget=Image_1
   properties {brush {
     bIsDynamicallyLoaded=false
     drawAs=Image
     imageType=FullColor
     mirroring=NoMirror
     resourceName=None
     resourceObject=/Game/UI/Frontend/ItemShop/MultiProduct/Art/T_UI__CrossEntitlement_RocketLeague
     tiling=NoTile
     imageSize=<256 256>
     margin=<ltrb 0 0 0 0>
     outlineSettings {
      bUseBrushTransparency=false
      roundingType=HalfHeightRadius
      width=0
      color {colorUseRule=UseColor_Specified, specifiedColor=<rgba 0 0 0 0>}
      cornerRadii=<0 0 0 0>
     }
     tintColor {colorUseRule=UseColor_Specified, specifiedColor=<rgba 1 1 1 1>}
     uVRegion {bIsValid=false, max=<0 0>, min=<0 0>}
    }}
   slot {class=CanvasPanelSlot, properties {layoutData {alignment=<0 0>, anchors {maximum=<0 0>, minimum=<0 0>}, offsets=<ltrb 524 20 100 90.5506>}}}
  }
  TextBlock_0 {
   widgetClassPath=&TextBlock
   parent=CanvasPanel_202
   widget=TextBlock_0
   properties {
    justification=Center
    text=Boosts should be approximately -6db on the Master Submix
    font {
     bForceMonospaced=false
     bMaterialIsStencil=false
     fontMaterial=null
     fontObject=&Burbank
     letterSpacing=0
     monospacedWidth=1
     size=15
     skewAmount=0
     typefaceFontName=Burbank-Black
     outlineSettings {
      bApplyOutlineToDropShadows=false
      bMiteredCorners=false
      bSeparateFillAlpha=false
      outlineMaterial=null
      outlineSize=0
      outlineColor=<rgba 0 0 0 1>
     }
    }
   }
   slot {class=CanvasPanelSlot, properties {layoutData {alignment=<0 0>, anchors {maximum=<0 0>, minimum=<0 0>}, offsets=<ltrb 232 368 151 40>}}}
  }
  TextBlock {
   widgetClassPath=&TextBlock
   parent=CanvasPanel_202
   widget=TextBlock
   properties {
    justification=Center
    text=Preview Section
    font {
     bForceMonospaced=false
     bMaterialIsStencil=false
     fontMaterial=null
     fontObject=&Roboto
     letterSpacing=0
     monospacedWidth=1
     size=15
     skewAmount=0
     typefaceFontName=Bold Italic
     outlineSettings {
      bApplyOutlineToDropShadows=false
      bMiteredCorners=false
      bSeparateFillAlpha=false
      outlineMaterial=null
      outlineSize=0
      outlineColor=<rgba 0 0 0 1>
     }
    }
   }
   slot {class=CanvasPanelSlot, properties {layoutData {alignment=<0 0>, anchors {maximum=<0 0>, minimum=<0 0>}, offsets=<ltrb 204 424 151 40>}}}
  }
  EditorUtilityButton_BoostStandardStart {
   widgetClassPath=&EditorUtilityButton
   parent=CanvasPanel_202
   widget=EditorUtilityButton_BoostStandardStart
   properties {widgetStyle {
     clickedSlateSound {resourceObject=null}
     disabled {
      bIsDynamicallyLoaded=false
      drawAs=RoundedBox
      imageType=NoImage
      mirroring=NoMirror
      resourceName=None
      resourceObject=null
      tiling=NoTile
      imageSize=<0 0>
      margin=<ltrb 0 0 0 0>
      outlineSettings {
       bUseBrushTransparency=false
       roundingType=FixedRadius
       width=1
       color {colorUseRule=UseColor_ColorTable, specifiedColor=<rgba 0 0 0 0>}
       cornerRadii=<4 4 4 4>
      }
      tintColor {colorUseRule=UseColor_Specified, specifiedColor=<rgba 0.0395 0.0395 0.0395 1>}
      uVRegion {bIsValid=false, max=<0 0>, min=<0 0>}
     }
     disabledForeground {colorUseRule=UseColor_ColorTable, specifiedColor=<rgba 0 0 0 0>}
     hovered {
      bIsDynamicallyLoaded=false
      drawAs=RoundedBox
      imageType=NoImage
      mirroring=NoMirror
      resourceName=None
      resourceObject=null
      tiling=NoTile
      imageSize=<32 32>
      margin=<ltrb 0 0 0 0>
      outlineSettings {
       bUseBrushTransparency=false
       roundingType=FixedRadius
       width=1
       color {colorUseRule=UseColor_ColorTable, specifiedColor=<rgba 0 0 0 0>}
       cornerRadii=<4 4 4 4>
      }
      tintColor {colorUseRule=UseColor_Specified, specifiedColor=<rgba 0.0953 0.0953 0.0953 1>}
      uVRegion {bIsValid=false, max=<0 0>, min=<0 0>}
     }
     hoveredForeground {colorUseRule=UseColor_ColorTable, specifiedColor=<rgba 0 0 0 0>}
     hoveredSlateSound {resourceObject=null}
     normal {
      bIsDynamicallyLoaded=false
      drawAs=RoundedBox
      imageType=NoImage
      mirroring=NoMirror
      resourceName=None
      resourceObject=null
      tiling=NoTile
      imageSize=<32 32>
      margin=<ltrb 0 0 0 0>
      outlineSettings {
       bUseBrushTransparency=false
       roundingType=FixedRadius
       width=1
       color {colorUseRule=UseColor_ColorTable, specifiedColor=<rgba 0 0 0 0>}
       cornerRadii=<4 4 4 4>
      }
      tintColor {colorUseRule=UseColor_Specified, specifiedColor=<rgba 0 0.4479 0.1008 1>}
      uVRegion {bIsValid=false, max=<0 0>, min=<0 0>}
     }
     normalForeground {colorUseRule=UseColor_ColorTable, specifiedColor=<rgba 0 0 0 0>}
     normalPadding=<ltrb 12 1.5 12 1.5>
     pressed {
      bIsDynamicallyLoaded=false
      drawAs=RoundedBox
      imageType=NoImage
      mirroring=NoMirror
      resourceName=None
      resourceObject=null
      tiling=NoTile
      imageSize=<32 32>
      margin=<ltrb 0 0 0 0>
      outlineSettings {
       bUseBrushTransparency=false
       roundingType=FixedRadius
       width=1
       color {colorUseRule=UseColor_ColorTable, specifiedColor=<rgba 0 0 0 0>}
       cornerRadii=<4 4 4 4>
      }
      tintColor {colorUseRule=UseColor_Specified, specifiedColor=<rgba 0.0284 0.0284 0.0284 1>}
      uVRegion {bIsValid=false, max=<0 0>, min=<0 0>}
     }
     pressedForeground {colorUseRule=UseColor_ColorTable, specifiedColor=<rgba 0 0 0 0>}
     pressedPadding=<ltrb 12 2.5 12 0.5>
     pressedSlateSound {resourceObject=null}
    }}
   slot {class=CanvasPanelSlot, properties {layoutData {alignment=<0 0>, anchors {maximum=<0 0>, minimum=<0 0>}, offsets=<ltrb 120 472 131.4973 40>}}}
  }
  UEFN_TextBlock_C_0 {
   widgetClassPath=&UEFN_TextBlock
   parent=EditorUtilityButton_BoostStandardStart
   widget=UEFN_TextBlock_C_0
   properties {
    text=Boost Standard Start
    font {
     bForceMonospaced=false
     bMaterialIsStencil=false
     fontMaterial=null
     fontObject=&BurbankSmall
     letterSpacing=0
     monospacedWidth=1
     size=7.5
     skewAmount=0
     typefaceFontName=Black
     outlineSettings {
      bApplyOutlineToDropShadows=false
      bMiteredCorners=false
      bSeparateFillAlpha=false
      outlineMaterial=null
      outlineSize=0
      outlineColor=<rgba 0 0 0 1>
     }
    }
   }
   slot {class=ButtonSlot}
  }
  EditorUtilityButton_BoostStandardLoop {
   widgetClassPath=&EditorUtilityButton
   parent=CanvasPanel_202
   widget=EditorUtilityButton_BoostStandardLoop
   properties {widgetStyle {
     clickedSlateSound {resourceObject=null}
     disabled {
      bIsDynamicallyLoaded=false
      drawAs=RoundedBox
      imageType=NoImage
      mirroring=NoMirror
      resourceName=None
      resourceObject=null
      tiling=NoTile
      imageSize=<0 0>
      margin=<ltrb 0 0 0 0>
      outlineSettings {
       bUseBrushTransparency=false
       roundingType=FixedRadius
       width=1
       color {colorUseRule=UseColor_ColorTable, specifiedColor=<rgba 0 0 0 0>}
       cornerRadii=<4 4 4 4>
      }
      tintColor {colorUseRule=UseColor_Specified, specifiedColor=<rgba 0.0395 0.0395 0.0395 1>}
      uVRegion {bIsValid=false, max=<0 0>, min=<0 0>}
     }
     disabledForeground {colorUseRule=UseColor_ColorTable, specifiedColor=<rgba 0 0 0 0>}
     hovered {
      bIsDynamicallyLoaded=false
      drawAs=RoundedBox
      imageType=NoImage
      mirroring=NoMirror
      resourceName=None
      resourceObject=null
      tiling=NoTile
      imageSize=<32 32>
      margin=<ltrb 0 0 0 0>
      outlineSettings {
       bUseBrushTransparency=false
       roundingType=FixedRadius
       width=1
       color {colorUseRule=UseColor_ColorTable, specifiedColor=<rgba 0 0 0 0>}
       cornerRadii=<4 4 4 4>
      }
      tintColor {colorUseRule=UseColor_Specified, specifiedColor=<rgba 0.0953 0.0953 0.0953 1>}
      uVRegion {bIsValid=false, max=<0 0>, min=<0 0>}
     }
     hoveredForeground {colorUseRule=UseColor_ColorTable, specifiedColor=<rgba 0 0 0 0>}
     hoveredSlateSound {resourceObject=null}
     normal {
      bIsDynamicallyLoaded=false
      drawAs=RoundedBox
      imageType=NoImage
      mirroring=NoMirror
      resourceName=None
      resourceObject=null
      tiling=NoTile
      imageSize=<32 32>
      margin=<ltrb 0 0 0 0>
      outlineSettings {
       bUseBrushTransparency=false
       roundingType=FixedRadius
       width=1
       color {colorUseRule=UseColor_ColorTable, specifiedColor=<rgba 0 0 0 0>}
       cornerRadii=<4 4 4 4>
      }
      tintColor {colorUseRule=UseColor_Specified, specifiedColor=<rgba 0 0.4479 0.1008 1>}
      uVRegion {bIsValid=false, max=<0 0>, min=<0 0>}
     }
     normalForeground {colorUseRule=UseColor_ColorTable, specifiedColor=<rgba 0 0 0 0>}
     normalPadding=<ltrb 12 1.5 12 1.5>
     pressed {
      bIsDynamicallyLoaded=false
      drawAs=RoundedBox
      imageType=NoImage
      mirroring=NoMirror
      resourceName=None
      resourceObject=null
      tiling=NoTile
      imageSize=<32 32>
      margin=<ltrb 0 0 0 0>
      outlineSettings {
       bUseBrushTransparency=false
       roundingType=FixedRadius
       width=1
       color {colorUseRule=UseColor_ColorTable, specifiedColor=<rgba 0 0 0 0>}
       cornerRadii=<4 4 4 4>
      }
      tintColor {colorUseRule=UseColor_Specified, specifiedColor=<rgba 0.0284 0.0284 0.0284 1>}
      uVRegion {bIsValid=false, max=<0 0>, min=<0 0>}
     }
     pressedForeground {colorUseRule=UseColor_ColorTable, specifiedColor=<rgba 0 0 0 0>}
     pressedPadding=<ltrb 12 2.5 12 0.5>
     pressedSlateSound {resourceObject=null}
    }}
   slot {class=CanvasPanelSlot, properties {layoutData {alignment=<0 0>, anchors {maximum=<0 0>, minimum=<0 0>}, offsets=<ltrb 328 472 229.9481 40>}}}
  }
  UEFN_TextBlock_C_5 {
   widgetClassPath=&UEFN_TextBlock
   parent=EditorUtilityButton_BoostStandardLoop
   widget=UEFN_TextBlock_C_5
   properties {
    text="Boost Standard Loop (HOLD TO DEMO)"
    font {
     bForceMonospaced=false
     bMaterialIsStencil=false
     fontMaterial=null
     fontObject=&BurbankSmall
     letterSpacing=0
     monospacedWidth=1
     size=7.5
     skewAmount=0
     typefaceFontName=Black
     outlineSettings {
      bApplyOutlineToDropShadows=false
      bMiteredCorners=false
      bSeparateFillAlpha=false
      outlineMaterial=null
      outlineSize=0
      outlineColor=<rgba 0 0 0 1>
     }
    }
   }
   slot {class=ButtonSlot}
  }
  EditorUtilityButton_56_MakeFolder {
   widgetClassPath=&EditorUtilityButton
   parent=CanvasPanel_202
   widget=EditorUtilityButton_56_MakeFolder
   properties {toolTipText=Make metasounds from selected soundwaves, backgroundColor=<rgba 0 0.1014 1 1>}
   slot {
    class=CanvasPanelSlot
    properties {layoutData {alignment=<0 0>, anchors {maximum=<0 0>, minimum=<0 0>}, offsets=<ltrb 48 304 272.6727 51.7718>}}
   }
  }
  UEFN_TextBlock_C_6 {
   widgetClassPath=&UEFN_TextBlock
   parent=EditorUtilityButton_56_MakeFolder
   widget=UEFN_TextBlock_C_6
   properties {
    justification=Center
    text=1. Make Folder
    font {
     bForceMonospaced=false
     bMaterialIsStencil=false
     fontMaterial=null
     fontObject=&BurbankSmall
     letterSpacing=0
     monospacedWidth=1
     size=9
     skewAmount=0
     typefaceFontName=Black
     outlineSettings {
      bApplyOutlineToDropShadows=false
      bMiteredCorners=false
      bSeparateFillAlpha=false
      outlineMaterial=null
      outlineSize=0
      outlineColor=<rgba 0 0 0 1>
     }
    }
   }
   slot {class=ButtonSlot}
  }
  EditorUtilityButton_BoostDVisualirectory {
   widgetClassPath=&EditorUtilityButton
   parent=CanvasPanel_202
   widget=EditorUtilityButton_BoostDVisualirectory
   properties {
    toolTipText=Set current editor path as the preset build destination
    backgroundColor=<rgba 1 0.0411 0.7089 1>
    widgetStyle {
     clickedSlateSound {resourceObject=null}
     disabled {
      bIsDynamicallyLoaded=false
      drawAs=RoundedBox
      imageType=NoImage
      mirroring=NoMirror
      resourceName=None
      resourceObject=null
      tiling=NoTile
      imageSize=<0 0>
      margin=<ltrb 0 0 0 0>
      outlineSettings {
       bUseBrushTransparency=false
       roundingType=FixedRadius
       width=1
       color {colorUseRule=UseColor_ColorTable, specifiedColor=<rgba 0 0 0 0>}
       cornerRadii=<4 4 4 4>
      }
      tintColor {colorUseRule=UseColor_Specified, specifiedColor=<rgba 0.0395 0.0395 0.0395 1>}
      uVRegion {bIsValid=false, max=<0 0>, min=<0 0>}
     }
     disabledForeground {colorUseRule=UseColor_ColorTable, specifiedColor=<rgba 0 0 0 0>}
     hovered {
      bIsDynamicallyLoaded=false
      drawAs=RoundedBox
      imageType=NoImage
      mirroring=NoMirror
      resourceName=None
      resourceObject=null
      tiling=NoTile
      imageSize=<32 32>
      margin=<ltrb 0 0 0 0>
      outlineSettings {
       bUseBrushTransparency=false
       roundingType=FixedRadius
       width=1
       color {colorUseRule=UseColor_ColorTable, specifiedColor=<rgba 0 0 0 0>}
       cornerRadii=<4 4 4 4>
      }
      tintColor {colorUseRule=UseColor_Specified, specifiedColor=<rgba 0.0953 0.0953 0.0953 1>}
      uVRegion {bIsValid=false, max=<0 0>, min=<0 0>}
     }
     hoveredForeground {colorUseRule=UseColor_ColorTable, specifiedColor=<rgba 0 0 0 0>}
     hoveredSlateSound {resourceObject=null}
     normal {
      bIsDynamicallyLoaded=false
      drawAs=RoundedBox
      imageType=NoImage
      mirroring=NoMirror
      resourceName=None
      resourceObject=null
      tiling=NoTile
      imageSize=<32 32>
      margin=<ltrb 0 0 0 0>
      outlineSettings {
       bUseBrushTransparency=false
       roundingType=FixedRadius
       width=1
       color {colorUseRule=UseColor_ColorTable, specifiedColor=<rgba 0 0 0 0>}
       cornerRadii=<4 4 4 4>
      }
      tintColor {colorUseRule=UseColor_Specified, specifiedColor=<rgba 0.0395 0.0395 0.0395 1>}
      uVRegion {bIsValid=false, max=<0 0>, min=<0 0>}
     }
     normalForeground {colorUseRule=UseColor_Specified, specifiedColor=<rgba 0 0 0 0>}
     normalPadding=<ltrb 12 1.5 12 1.5>
     pressed {
      bIsDynamicallyLoaded=false
      drawAs=RoundedBox
      imageType=NoImage
      mirroring=NoMirror
      resourceName=None
      resourceObject=null
      tiling=NoTile
      imageSize=<32 32>
      margin=<ltrb 0 0 0 0>
      outlineSettings {
       bUseBrushTransparency=false
       roundingType=FixedRadius
       width=1
       color {colorUseRule=UseColor_ColorTable, specifiedColor=<rgba 0 0 0 0>}
       cornerRadii=<4 4 4 4>
      }
      tintColor {colorUseRule=UseColor_Specified, specifiedColor=<rgba 0.0284 0.0284 0.0284 1>}
      uVRegion {bIsValid=false, max=<0 0>, min=<0 0>}
     }
     pressedForeground {colorUseRule=UseColor_ColorTable, specifiedColor=<rgba 0 0 0 0>}
     pressedPadding=<ltrb 12 2.5 12 0.5>
     pressedSlateSound {resourceObject=null}
    }
   }
   slot {
    class=CanvasPanelSlot
    properties {layoutData {alignment=<0 0>, anchors {maximum=<0 0>, minimum=<0 0>}, offsets=<ltrb 696 188 200.9009 72.6426>}}
   }
  }
  UEFN_TextBlock_C_8 {
   widgetClassPath=&UEFN_TextBlock
   parent=EditorUtilityButton_BoostDVisualirectory
   widget=UEFN_TextBlock_C_8
   properties {
    justification=Center
    text=Boost Visuals Directory
    font {
     bForceMonospaced=false
     bMaterialIsStencil=false
     fontMaterial=null
     fontObject=&Roboto
     letterSpacing=0
     monospacedWidth=1
     size=9
     skewAmount=0
     typefaceFontName=Regular
     outlineSettings {
      bApplyOutlineToDropShadows=false
      bMiteredCorners=false
      bSeparateFillAlpha=false
      outlineMaterial=null
      outlineSize=0
      outlineColor=<rgba 0 0 0 1>
     }
    }
   }
   slot {class=ButtonSlot}
  }
 ]
}
